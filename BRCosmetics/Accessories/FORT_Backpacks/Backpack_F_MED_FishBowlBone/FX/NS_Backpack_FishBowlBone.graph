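# Niagara effect for the FishBowlBone back bling that flashes lightning arcs and small lights attached to the backpack's skeletal mesh.
# Four emitters loop with random delays, a fifth fires a delayed burst, and a test-like sixth is disabled.
# Each places meshes via skeletal sampling, mostly choosing between two orientations, and animates glow and dynamic material parameters.
# Shared system color, lifetime and curves drive the look, with brightness scaled by time of day.
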NiagaraSystem /BRCosmetics/Accessories/FORT_Backpacks/Backpack_F_MED_FishBowlBone/FX/NS_Backpack_FishBowlBone
dependencies {
 usedDataInterfaces [&NiagaraDataInterfaceCurve]
 usedDynamicInputs [&FNTimeOfDayScalar, &FN_FrontEndFloatModifier]
 usedModules [
  &self:SystemSpawnScript.SystemScriptSource.SystemScriptGraph.NiagaraNodeAssignment_0.SetVariables_#1
  &self:SystemSpawnScript.SystemScriptSource.SystemScriptGraph.NiagaraNodeAssignment_1.SetVariables_#2
  &self:SystemSpawnScript.SystemScriptSource.SystemScriptGraph.NiagaraNodeAssignment_3.SetVariables_#3
  &self:SystemSpawnScript.SystemScriptSource.SystemScriptGraph.NiagaraNodeAssignment_2.SetVariables_#4
  &self:SystemSpawnScript.SystemScriptSource.SystemScriptGraph.NiagaraNodeAssignment_4.SetVariables_#5
  &SystemState
 ]
}
emitters {
 Lightning01 {
  emitterUpdateScript [
   &EmitterState(
    Life Cycle Mode=::NewEnumerator1
    Loop Delay=&RandomRangeFloat(Minimum=2, Maximum=5)
    Loop Duration=&RandomRangeFloat(Minimum=0.3, Maximum=0.5)
    Recalculate Duration Each Loop=true
    UseLoopDelay=true
   )
   &SpawnBurst_Instantaneous(Use Spawn Probability=false, UseLoopCountLimit=false)
   &SpawnRate(SpawnRate=0, Use Spawn Probability=false) {enabled=false}
  ]
  particleSpawnScript [
   /SetVariables(Particles.RandomFloat=&RandomRangeFloat(Minimum=0.5, Maximum=1)) {isSetParameters=true, module=SetVariables_#6}
   &InitializeParticle(
    Color=$System.LightningColor
    Lifetime=&Multiply_Float(A=$System.LightningLifetime, B=$Particles.RandomFloat)
    Mesh Scale Mode=::NewEnumerator4
    UsePositionOffset=false
    Write Material Random=true
   )
   &InitialMeshOrientation(
    Mesh Orientation Mode=::NewEnumerator2
    Rotation=&MakeCustomVectorFromBool(Bool=&RandomBool(), True Vector=<0 0.05 -0.1>, False Vector=<0 0.05 0.5>)
    Use Rotation Vector=true
   )
   &CameraOffset(Camera Offset Amount=3)
  ]
  particleUpdateScript [
   &SkeletalMeshLocation(Bone Sampling Mode=::NewEnumerator4, MaskSpawnGroup=false, Skeletal Mesh=$User.Skeletal Mesh)
   &ParticleState()
   &ScaleMeshSize(Scale Factor=&Multiply_VectorByFloat(Vector=<1 1 1>, Float=&RandomRangeFloat(Minimum=0.8, Maximum=1)))
   &ScaleColor(
    Scale Alpha=&Multiply_Float(A=20, B=&FloatFromCurve(FloatCurve=$System.AlphaCurve))
    Scale RGB=&Multiply_VectorByFloat(
     Vector=<1 1 1>
     Float=&Multiply_Float(
      A=&FloatFromCurve(FloatCurve=$System.EmissiveCurve)
      B=&Multiply_Float(A=$System.Tod, B=1)
     )
    )
    ScaleA=true
    ScaleRGB=true
   )
   &DynamicMaterialParameters(
    Index 0 Param 1=&FloatFromCurve(FloatCurve=DataInterface() {propertyValues {Curve {keys [
         {arriveTangent=0.4234, interpMode=RCIM_Cubic, leaveTangent=0.4234, tangentMode=RCTM_User, time=0, value=0}
         {arriveTangent=2.3288, interpMode=RCIM_Cubic, leaveTangent=2.3288, tangentMode=RCTM_User, time=1, value=1}
        ]}}})
    Index 0 Param 2=0.7
    Index 0 Param 3=&FloatFromCurve(FloatCurve=DataInterface() {propertyValues {Curve {keys [{tangentMode=RCTM_User, time=0, value=0}, {tangentMode=RCTM_User, time=1, value=5}]}}})
    Index 0 Param 4=0.7
    Index 1 Param 1=&FloatFromCurve(FloatCurve=DataInterface() {propertyValues {Curve {keys [<at 0.5 1>,<at 1 0.3>]}}}, Scale Curve=0.3)
    Index 1 Param 3=$System.Lightning_U_Tile
    Index 1 Param 4=&FloatFromCurve(FloatCurve=DataInterface() {propertyValues {Curve {keys [<at 0.3 1>,<at 1 0>]}}})
    Module.Param0WriteEnabled=true
    Module.Param0WriteEnabled1=true
    Module.Param0WriteEnabled2=true
    Module.Param1WriteEnabled=true
    Module.Param1WriteEnabled1=true
    Module.Param1WriteEnabled2=true
    Module.Param2WriteEnabled=true
    Module.Param2WriteEnabled1=true
    Module.Param2WriteEnabled2=true
    Module.Param3WriteEnabled=true
    Module.Param3WriteEnabled1=true
    Module.Param3WriteEnabled2=true
    Write Parameter Index 1=true
    Write Parameter Index 2=true
   ) {enabled=false, module=DynamicMaterialParameters001}
   &DynamicMaterialParameters(
    Index 0 Param 1=1
    Index 0 Param 2=0.7
    Index 0 Param 4=0.7
    Index 1 Param 1=&FloatFromCurve(FloatCurve=DataInterface() {propertyValues {Curve {keys [<at 0.5 1>,<at 1 0.3>]}}}, Scale Curve=0.3)
    Index 1 Param 3=$System.Lightning_U_Tile
    Index 1 Param 4=1
    Index 2 Param 1=&FloatFromCurve(FloatCurve=$System.LeadEdgeCurve)
    Module.Param0WriteEnabled=true
    Module.Param0WriteEnabled1=true
    Module.Param0WriteEnabled2=true
    Module.Param1WriteEnabled=true
    Module.Param1WriteEnabled1=true
    Module.Param1WriteEnabled2=true
    Module.Param2WriteEnabled=true
    Module.Param2WriteEnabled1=true
    Module.Param2WriteEnabled2=true
    Module.Param3WriteEnabled=true
    Module.Param3WriteEnabled1=true
    Module.Param3WriteEnabled2=true
    Write Parameter Index 1=true
    Write Parameter Index 2=true
   )
   &SolveForcesAndVelocity(Clamp Velocity=false, Limit Acceleration=false)
   /SetVariables(Particles.Glow Color=&MakeLinearColorFromVectorAndFloat(
     "Vector (RGB)"=&MakeVectorFromLinearColorRGB(Color=&MultiplyLinearColor(Linear Color=<rgba 0.5 0.65 1 1>, Scale Factor=&MakeVector4_FromFloat(Float=1)))
     "Float (Alpha)"=&Multiply_Float(
      A=&FloatFromCurve(
       FloatCurve=DataInterface() {propertyValues {Curve {keys [
           {interpMode=RCIM_Cubic, tangentMode=RCTM_User, time=0, value=1}
           <at 0.7 1>
           {arriveTangent=-3.6307, interpMode=RCIM_Cubic, leaveTangent=-3.6307, tangentMode=RCTM_User, time=0.9, value=0}
          ]}}}
       Scale Curve=$System.Tod
      )
      B=1.5
     )
    )) {isSetParameters=true, module=SetVariables_#7}
   /SetVariables(Particles.Glow Dy02:Vector4f=&MakeVector4(
     X=&FloatFromCurve(FloatCurve=DataInterface() {propertyValues {Curve {keys [
          {arriveTangent=0.4234, interpMode=RCIM_Cubic, leaveTangent=0.4234, tangentMode=RCTM_User, time=0, value=0}
          {arriveTangent=2.3288, interpMode=RCIM_Cubic, leaveTangent=2.3288, tangentMode=RCTM_User, time=0.86, value=1}
         ]}}})
     Y=0.7
     Z=0
     W=0.5
    )) {isSetParameters=true, module=SetVariables_#8}
   /SetVariables(Particles.Glow Dy03:Vector4f=&MakeVector4(
     X=&FloatFromCurve(FloatCurve=DataInterface() {propertyValues {Curve {keys [<at 0 1>,<at 1 0.5>]}}}, Scale Curve=0.5)
     Y=0
     Z=0
     W=&FloatFromCurve(FloatCurve=DataInterface() {propertyValues {Curve {keys [<at 0.5 1>,<at 1 0>]}}})
    )) {isSetParameters=true, module=SetVariables_#9}
   &FN_Light_Attributes(
    Color=$System.ParticleLightColor
    Create Color Attribute=true
    Create Exponent Attribute=true
    Create Position Attribute=true
    Create Radius Attribute=true
    Exponent=2
    Position=&AddVectorToPosition(Position=$Particles.Position, Vector=&RandomRangeVector(Minimum=<5 -8 0>, Maximum=<-5 -8 5>))
    Radius=20
    Scale Color=&FloatFromCurve(FloatCurve=$System.ParticleLightCurve, Scale Curve=$System.Tod)
   )
  ]
  rendererClasses [&NiagaraMeshRendererProperties, &NiagaraLightRendererProperties]
  renderers [
   {rendererClass=&NiagaraMeshRendererProperties, rendererIndex=0}
   {rendererClass=&NiagaraMeshRendererProperties, rendererIndex=1}
   {rendererClass=&NiagaraLightRendererProperties, rendererIndex=2}
  ]
 }
 Lightning01001 {
  bEnabled=false
  emitterUpdateScript [
   &EmitterState()
   &SpawnBurst_Instantaneous(Use Spawn Probability=false, UseLoopCountLimit=false) {enabled=false}
   &SpawnRate(SpawnRate=2, Use Spawn Probability=false)
  ]
  particleSpawnScript [
   &InitializeParticle(Color=<rgba 0.5 0.65 1 1>, Lifetime=1.5, Mesh Scale Mode=::NewEnumerator4, UsePositionOffset=false, Write Material Random=true)
   &InitialMeshOrientation(
    Mesh Orientation Mode=::NewEnumerator2
    Rotation=&MakeCustomVectorFromBool(Bool=&RandomBool(), True Vector=<0 0.05 -0.1>, False Vector=<0 0.05 0.5>)
    Use Rotation Vector=true
   )
   &CameraOffset(Camera Offset Amount=3)
  ]
  particleUpdateScript [
   &SkeletalMeshLocation(Bone Sampling Mode=::NewEnumerator4, MaskSpawnGroup=false, Skeletal Mesh=$User.Skeletal Mesh)
   &ParticleState()
   &ScaleMeshSize(Scale Factor=&Multiply_VectorByFloat(Vector=<1 1 1>, Float=&RandomRangeFloat(Minimum=0.8, Maximum=1)))
   &ScaleColor(
    Scale Alpha=&Multiply_Float(A=20, B=&FloatFromCurve(FloatCurve=$System.AlphaCurve))
    Scale RGB=&Multiply_VectorByFloat(
     Vector=<1 1 1>
     Float=&Multiply_Float(
      A=&FloatFromCurve(FloatCurve=$System.EmissiveCurve)
      B=&Multiply_Float(A=$System.Tod, B=1)
     )
    )
    ScaleA=true
    ScaleRGB=true
   ) {enabled=false}
   &SolveForcesAndVelocity(Clamp Velocity=false, Limit Acceleration=false)
   &DynamicMaterialParameters(
    Index 0 Param 1=1
    Index 0 Param 2=0.7
    Index 0 Param 4=0.7
    Index 1 Param 1=&FloatFromCurve(FloatCurve=DataInterface() {propertyValues {Curve {keys [<at 0.5 1>,<at 1 0.3>]}}}, Scale Curve=0.3)
    Index 1 Param 3=$System.Lightning_U_Tile
    Index 1 Param 4=1
    Index 2 Param 1=&FloatFromCurve(FloatCurve=DataInterface() {propertyValues {Curve {keys [{tangentMode=RCTM_User, time=0, value=0}, {tangentMode=RCTM_User, time=1, value=1}]}}})
    Module.Param0WriteEnabled=true
    Module.Param0WriteEnabled1=true
    Module.Param0WriteEnabled2=true
    Module.Param1WriteEnabled=true
    Module.Param1WriteEnabled1=true
    Module.Param1WriteEnabled2=true
    Module.Param2WriteEnabled=true
    Module.Param2WriteEnabled1=true
    Module.Param2WriteEnabled2=true
    Module.Param3WriteEnabled=true
    Module.Param3WriteEnabled1=true
    Module.Param3WriteEnabled2=true
    Write Parameter Index 1=true
    Write Parameter Index 2=true
   ) {module=DynamicMaterialParameters001}
   /SetVariables(Particles.Glow Color=&MakeLinearColorFromVectorAndFloat(
     "Vector (RGB)"=&MakeVectorFromLinearColorRGB(Color=&MultiplyLinearColor(Linear Color=<rgba 0.5 0.65 1 1>, Scale Factor=&MakeVector4_FromFloat(Float=1)))
     "Float (Alpha)"=&Multiply_Float(
      A=&FloatFromCurve(
       FloatCurve=DataInterface() {propertyValues {Curve {keys [
           {interpMode=RCIM_Cubic, tangentMode=RCTM_User, time=0, value=1}
           <at 0.7 1>
           {arriveTangent=-3.6307, interpMode=RCIM_Cubic, leaveTangent=-3.6307, tangentMode=RCTM_User, time=0.9, value=0}
          ]}}}
       Scale Curve=$System.Tod
      )
      B=1.5
     )
    )) {isSetParameters=true, module=SetVariables_#7}
   /SetVariables(Particles.Glow Dy02:Vector4f=&MakeVector4(
     X=&FloatFromCurve(FloatCurve=DataInterface() {propertyValues {Curve {keys [
          {arriveTangent=0.4234, interpMode=RCIM_Cubic, leaveTangent=0.4234, tangentMode=RCTM_User, time=0, value=0}
          {arriveTangent=2.3288, interpMode=RCIM_Cubic, leaveTangent=2.3288, tangentMode=RCTM_User, time=0.86, value=1}
         ]}}})
     Y=0.7
     Z=0
     W=0.5
    )) {isSetParameters=true, module=SetVariables_#8}
   /SetVariables(Particles.Glow Dy03:Vector4f=&MakeVector4(
     X=&FloatFromCurve(FloatCurve=DataInterface() {propertyValues {Curve {keys [<at 0 1>,<at 1 0.5>]}}}, Scale Curve=0.5)
     Y=0
     Z=0
     W=&FloatFromCurve(FloatCurve=DataInterface() {propertyValues {Curve {keys [<at 0.5 1>,<at 1 0>]}}})
    )) {isSetParameters=true, module=SetVariables_#9}
   &FN_Light_Attributes(
    Color=<rgba 2.5 3.25 5 1>
    Create Color Attribute=true
    Create Exponent Attribute=true
    Create Position Attribute=true
    Create Radius Attribute=true
    Exponent=2
    Position=&AddVectorToPosition(Position=$Particles.Position, Vector=&RandomRangeVector(Minimum=<5 -8 0>, Maximum=<-5 -8 5>))
    Radius=20
    Scale Color=&FloatFromCurve(FloatCurve=DataInterface() {propertyValues {Curve {keys [<at 0 0>,<at 0.15 1>,<at 0.5 1>,<at 1 0>]}}}, Scale Curve=$System.Tod)
   )
  ]
  rendererClasses [&NiagaraMeshRendererProperties, &NiagaraLightRendererProperties]
  renderers [
   {rendererClass=&NiagaraMeshRendererProperties, rendererIndex=0}
   {rendererClass=&NiagaraMeshRendererProperties, rendererIndex=1}
   {rendererClass=&NiagaraLightRendererProperties, rendererIndex=2}
  ]
 }
 Lightning02 {
  emitterUpdateScript [
   &EmitterState(
    Life Cycle Mode=::NewEnumerator1
    Loop Delay=&RandomRangeFloat(Minimum=2, Maximum=5)
    Loop Duration=&RandomRangeFloat(Minimum=0.3, Maximum=0.5)
    Recalculate Duration Each Loop=true
    UseLoopDelay=true
   )
   &SpawnRate(Use Spawn Probability=false)
  ]
  particleSpawnScript [
   /SetVariables(Particles.RandomFloat=&RandomRangeFloat(Minimum=0.5, Maximum=1)) {isSetParameters=true, module=SetVariables_#6}
   &InitializeParticle(
    Color=$System.LightningColor
    Lifetime=&Multiply_Float(A=$System.LightningLifetime, B=$Particles.RandomFloat)
    Mesh Scale=<1 1 0.9>
    Mesh Scale Mode=::NewEnumerator4
    UsePositionOffset=false
    Write Material Random=true
   )
   &InitialMeshOrientation(
    Mesh Orientation Mode=::NewEnumerator2
    Rotation=&MakeCustomVectorFromBool(Bool=&RandomBool(), True Vector=<-0.2 0.5 -0.2>, False Vector=<-0.2 0.2 0>)
    Use Rotation Vector=true
   )
   &CameraOffset(Camera Offset Amount=5)
  ]
  particleUpdateScript [
   &SkeletalMeshLocation(Bone Sampling Mode=::NewEnumerator4, MaskSpawnGroup=false, Sampled Position Offset=<-2 -5 0>, Skeletal Mesh=$User.Skeletal Mesh)
   &ParticleState()
   &ScaleMeshSize(Scale Factor=&Multiply_VectorByFloat(Vector=<1 1 1>, Float=&RandomRangeFloat(Minimum=0.8, Maximum=1)))
   &ScaleColor(
    Scale Alpha=&Multiply_Float(A=20, B=&FloatFromCurve(FloatCurve=$System.AlphaCurve))
    Scale RGB=&Multiply_VectorByFloat(
     Vector=<1 1 1>
     Float=&Multiply_Float(
      A=&FloatFromCurve(FloatCurve=$System.EmissiveCurve)
      B=&Multiply_Float(A=$System.Tod, B=1)
     )
    )
    ScaleA=true
    ScaleRGB=true
   )
   &DynamicMaterialParameters(
    Index 0 Param 1=&FloatFromCurve(FloatCurve=DataInterface() {propertyValues {Curve {keys [
         {arriveTangent=0.4234, interpMode=RCIM_Cubic, leaveTangent=0.4234, tangentMode=RCTM_User, time=0, value=0}
         {arriveTangent=2.3288, interpMode=RCIM_Cubic, leaveTangent=2.3288, tangentMode=RCTM_User, time=1, value=1}
        ]}}})
    Index 0 Param 2=0.7
    Index 0 Param 3=&FloatFromCurve(FloatCurve=DataInterface() {propertyValues {Curve {keys [{tangentMode=RCTM_User, time=0, value=0}, {tangentMode=RCTM_User, time=1, value=5}]}}})
    Index 0 Param 4=0.7
    Index 1 Param 1=&FloatFromCurve(FloatCurve=DataInterface() {propertyValues {Curve {keys [<at 0 1>,<at 1 0.3>]}}}, Scale Curve=0.3)
    Index 1 Param 3=$System.Lightning_U_Tile
    Index 1 Param 4=&FloatFromCurve(FloatCurve=DataInterface() {propertyValues {Curve {keys [<at 0.5 1>,<at 1 0>]}}})
    Module.Param0WriteEnabled=true
    Module.Param0WriteEnabled1=true
    Module.Param1WriteEnabled=true
    Module.Param1WriteEnabled1=true
    Module.Param2WriteEnabled=true
    Module.Param2WriteEnabled1=true
    Module.Param3WriteEnabled=true
    Module.Param3WriteEnabled1=true
    Write Parameter Index 1=true
   ) {enabled=false}
   &DynamicMaterialParameters(
    Index 0 Param 1=1
    Index 0 Param 2=0.7
    Index 0 Param 4=0.7
    Index 1 Param 1=&FloatFromCurve(FloatCurve=DataInterface() {propertyValues {Curve {keys [<at 0.5 1>,<at 1 0.3>]}}}, Scale Curve=0.3)
    Index 1 Param 3=$System.Lightning_U_Tile
    Index 1 Param 4=1
    Index 2 Param 1=&FloatFromCurve(FloatCurve=$System.LeadEdgeCurve)
    Module.Param0WriteEnabled=true
    Module.Param0WriteEnabled1=true
    Module.Param0WriteEnabled2=true
    Module.Param1WriteEnabled=true
    Module.Param1WriteEnabled1=true
    Module.Param1WriteEnabled2=true
    Module.Param2WriteEnabled=true
    Module.Param2WriteEnabled1=true
    Module.Param2WriteEnabled2=true
    Module.Param3WriteEnabled=true
    Module.Param3WriteEnabled1=true
    Module.Param3WriteEnabled2=true
    Write Parameter Index 1=true
    Write Parameter Index 2=true
   ) {module=DynamicMaterialParameters001}
   &SolveForcesAndVelocity(Clamp Velocity=false, Limit Acceleration=false)
   /SetVariables(Particles.Glow Color=&MakeLinearColorFromVectorAndFloat(
     "Vector (RGB)"=&MakeVectorFromLinearColorRGB(Color=&MultiplyLinearColor(Linear Color=<rgba 0.5 0.65 1 1>, Scale Factor=&MakeVector4_FromFloat(Float=1)))
     "Float (Alpha)"=&Multiply_Float(
      A=&FloatFromCurve(
       FloatCurve=DataInterface() {propertyValues {Curve {keys [
           {interpMode=RCIM_Cubic, tangentMode=RCTM_User, time=0, value=1}
           <at 0.7 1>
           {arriveTangent=-3.6307, interpMode=RCIM_Cubic, leaveTangent=-3.6307, tangentMode=RCTM_User, time=0.9, value=0}
          ]}}}
       Scale Curve=$System.Tod
      )
      B=1.5
     )
    )) {isSetParameters=true}
   /SetVariables(Particles.Glow Dy02:Vector4f=&MakeVector4(
     X=&FloatFromCurve(FloatCurve=DataInterface() {propertyValues {Curve {keys [
          {arriveTangent=0.4234, interpMode=RCIM_Cubic, leaveTangent=0.4234, tangentMode=RCTM_User, time=0, value=0}
          {arriveTangent=2.3288, interpMode=RCIM_Cubic, leaveTangent=2.3288, tangentMode=RCTM_User, time=0.9, value=1}
         ]}}})
     Y=0.7
     Z=0
     W=0.5
    )) {isSetParameters=true}
   /SetVariables(Particles.Glow Dy03:Vector4f=&MakeVector4(
     X=&FloatFromCurve(FloatCurve=DataInterface() {propertyValues {Curve {keys [<at 0 1>,<at 1 0.5>]}}}, Scale Curve=0.5)
     Y=0
     Z=0
     W=&FloatFromCurve(FloatCurve=DataInterface() {propertyValues {Curve {keys [<at 0.5 1>,<at 1 0>]}}})
    )) {isSetParameters=true}
   &FN_Light_Attributes(
    Color=$System.ParticleLightColor
    Create Color Attribute=true
    Create Exponent Attribute=true
    Create Position Attribute=true
    Create Radius Attribute=true
    Exponent=2
    Position=&AddVectorToPosition(Position=$Particles.Position, Vector=&RandomRangeVector(Minimum=<10 -8 0>, Maximum=<-10 -8 5>))
    Radius=30
    Scale Color=&FloatFromCurve(FloatCurve=$System.ParticleLightCurve, Scale Curve=$System.Tod)
   )
  ]
  rendererClasses [&NiagaraMeshRendererProperties, &NiagaraLightRendererProperties]
  renderers [
   {rendererClass=&NiagaraMeshRendererProperties, rendererIndex=0}
   {rendererClass=&NiagaraMeshRendererProperties, rendererIndex=1}
   {rendererClass=&NiagaraLightRendererProperties, rendererIndex=2}
  ]
 }
 Lightning03 {
  emitterUpdateScript [
   &EmitterState(
    Life Cycle Mode=::NewEnumerator1
    Loop Delay=&RandomRangeFloat(Minimum=2, Maximum=5)
    Loop Duration=&RandomRangeFloat(Minimum=0.3, Maximum=0.5)
    Recalculate Duration Each Loop=true
    UseLoopDelay=true
   )
   &SpawnRate(Use Spawn Probability=false)
  ]
  particleSpawnScript [
   /SetVariables(Particles.RandomFloat=&RandomRangeFloat(Minimum=0.5, Maximum=1)) {isSetParameters=true, module=SetVariables_#6}
   &InitializeParticle(
    Color=$System.LightningColor
    Lifetime=&Multiply_Float(A=$System.LightningLifetime, B=$Particles.RandomFloat)
    Mesh Scale Mode=::NewEnumerator4
    UsePositionOffset=false
    Write Material Random=true
   )
   &InitialMeshOrientation(
    Mesh Orientation Mode=::NewEnumerator2
    Rotation=&MakeCustomVectorFromBool(Bool=&RandomBool(), True Vector=<0 -0.2 0.6>, False Vector=<0.5 -0.2 0.2>)
    Use Rotation Vector=true
   )
   &CameraOffset(Camera Offset Amount=5)
  ]
  particleUpdateScript [
   &SkeletalMeshLocation(Bone Sampling Mode=::NewEnumerator4, MaskSpawnGroup=false, Skeletal Mesh=$User.Skeletal Mesh)
   &ParticleState()
   &ScaleMeshSize(Scale Factor=&Multiply_VectorByFloat(Vector=<1 1 1>, Float=&RandomRangeFloat(Minimum=0.8, Maximum=1)))
   &ScaleColor(
    Scale Alpha=&Multiply_Float(A=20, B=&FloatFromCurve(FloatCurve=$System.AlphaCurve))
    Scale RGB=&Multiply_VectorByFloat(
     Vector=<1 1 1>
     Float=&Multiply_Float(
      A=&FloatFromCurve(FloatCurve=$System.EmissiveCurve)
      B=&Multiply_Float(A=$System.Tod, B=1)
     )
    )
    ScaleA=true
    ScaleRGB=true
   )
   &DynamicMaterialParameters(
    Index 0 Param 1=&FloatFromCurve(FloatCurve=DataInterface() {propertyValues {Curve {keys [
         {arriveTangent=0.4234, interpMode=RCIM_Cubic, leaveTangent=0.4234, tangentMode=RCTM_User, time=0, value=0}
         {arriveTangent=2.3288, interpMode=RCIM_Cubic, leaveTangent=2.3288, tangentMode=RCTM_User, time=1, value=1}
        ]}}})
    Index 0 Param 2=0.7
    Index 0 Param 3=&FloatFromCurve(FloatCurve=DataInterface() {propertyValues {Curve {keys [{tangentMode=RCTM_User, time=0, value=0}, {tangentMode=RCTM_User, time=1, value=5}]}}})
    Index 0 Param 4=0.7
    Index 1 Param 1=&FloatFromCurve(FloatCurve=DataInterface() {propertyValues {Curve {keys [<at 0 1>,<at 1 0.3>]}}}, Scale Curve=0.3)
    Index 1 Param 3=$System.Lightning_U_Tile
    Module.Param0WriteEnabled=true
    Module.Param0WriteEnabled1=true
    Module.Param1WriteEnabled=true
    Module.Param1WriteEnabled1=true
    Module.Param2WriteEnabled=true
    Module.Param2WriteEnabled1=true
    Module.Param3WriteEnabled=true
    Module.Param3WriteEnabled1=true
    Write Parameter Index 1=true
   ) {enabled=false}
   &DynamicMaterialParameters(
    Index 0 Param 1=1
    Index 0 Param 2=0.7
    Index 0 Param 4=0.7
    Index 1 Param 1=&FloatFromCurve(FloatCurve=DataInterface() {propertyValues {Curve {keys [<at 0.5 1>,<at 1 0.3>]}}}, Scale Curve=0.3)
    Index 1 Param 3=$System.Lightning_U_Tile
    Index 1 Param 4=1
    Index 2 Param 1=&FloatFromCurve(FloatCurve=$System.LeadEdgeCurve)
    Module.Param0WriteEnabled=true
    Module.Param0WriteEnabled1=true
    Module.Param0WriteEnabled2=true
    Module.Param1WriteEnabled=true
    Module.Param1WriteEnabled1=true
    Module.Param1WriteEnabled2=true
    Module.Param2WriteEnabled=true
    Module.Param2WriteEnabled1=true
    Module.Param2WriteEnabled2=true
    Module.Param3WriteEnabled=true
    Module.Param3WriteEnabled1=true
    Module.Param3WriteEnabled2=true
    Write Parameter Index 1=true
    Write Parameter Index 2=true
   ) {module=DynamicMaterialParameters001}
   &SolveForcesAndVelocity(Clamp Velocity=false, Limit Acceleration=false)
   /SetVariables(Particles.Glow Color=&MakeLinearColorFromVectorAndFloat(
     "Vector (RGB)"=&MakeVectorFromLinearColorRGB(Color=&MultiplyLinearColor(Linear Color=<rgba 0.5 0.65 1 1>, Scale Factor=&MakeVector4_FromFloat(Float=1)))
     "Float (Alpha)"=&Multiply_Float(
      A=&FloatFromCurve(
       FloatCurve=DataInterface() {propertyValues {Curve {keys [
           {interpMode=RCIM_Cubic, tangentMode=RCTM_User, time=0, value=1}
           <at 0.7 1>
           {arriveTangent=-3.6307, interpMode=RCIM_Cubic, leaveTangent=-3.6307, tangentMode=RCTM_User, time=0.9, value=0}
          ]}}}
       Scale Curve=$System.Tod
      )
      B=1.5
     )
    )) {isSetParameters=true, module=SetVariables_#10}
   /SetVariables(Particles.Glow Dy02:Vector4f=&MakeVector4(
     X=&FloatFromCurve(FloatCurve=DataInterface() {propertyValues {Curve {keys [
          {arriveTangent=0.4234, interpMode=RCIM_Cubic, leaveTangent=0.4234, tangentMode=RCTM_User, time=0, value=0}
          {arriveTangent=2.3288, interpMode=RCIM_Cubic, leaveTangent=2.3288, tangentMode=RCTM_User, time=0.9, value=1}
         ]}}})
     Y=0.7
     Z=0
     W=0.5
    )) {isSetParameters=true, module=SetVariables_#11}
   /SetVariables(Particles.Glow Dy03:Vector4f=&MakeVector4(
     X=&FloatFromCurve(FloatCurve=DataInterface() {propertyValues {Curve {keys [<at 0 1>,<at 1 0.5>]}}}, Scale Curve=0.5)
     Y=0
     Z=0
     W=0
    )) {isSetParameters=true, module=SetVariables_#12}
   &FN_Light_Attributes(
    Color=$System.ParticleLightColor
    Create Color Attribute=true
    Create Exponent Attribute=true
    Create Position Attribute=true
    Create Radius Attribute=true
    Exponent=2
    Position=&AddVectorToPosition(Position=$Particles.Position, Vector=&RandomRangeVector(Minimum=<5 -8 0>, Maximum=<-5 -8 5>))
    Radius=30
    Scale Color=&FloatFromCurve(FloatCurve=$System.ParticleLightCurve, Scale Curve=$System.Tod)
   )
  ]
  rendererClasses [&NiagaraMeshRendererProperties, &NiagaraLightRendererProperties]
  renderers [
   {rendererClass=&NiagaraMeshRendererProperties, rendererIndex=0}
   {rendererClass=&NiagaraMeshRendererProperties, rendererIndex=1}
   {rendererClass=&NiagaraLightRendererProperties, rendererIndex=2}
  ]
 }
 Lightning03001 {
  emitterUpdateScript [
   &EmitterState(Life Cycle Mode=::NewEnumerator1, Loop Behavior=::NewEnumerator1, Loop Delay=1, Loop Duration=1, UseLoopDelay=true)
   &SpawnBurst_Instantaneous(Use Spawn Probability=false, UseLoopCountLimit=false)
   &SpawnRate(Use Spawn Probability=false) {enabled=false}
  ]
  particleSpawnScript [
   /SetVariables(Particles.RandomFloat=&RandomRangeFloat(Minimum=0.5, Maximum=1)) {isSetParameters=true, module=SetVariables_#6}
   &InitializeParticle(
    Color=$System.LightningColor
    Lifetime=&Multiply_Float(A=$System.LightningLifetime, B=$Particles.RandomFloat)
    Mesh Scale Mode=::NewEnumerator4
    UsePositionOffset=false
    Write Material Random=true
   )
   &InitialMeshOrientation(
    Mesh Orientation Mode=::NewEnumerator2
    Rotation=&MakeCustomVectorFromBool(Bool=true, True Vector=<0 -0.2 0.6>, False Vector=<0.5 -0.2 0.2>)
    Use Rotation Vector=true
   )
   &CameraOffset(Camera Offset Amount=5)
  ]
  particleUpdateScript [
   &SkeletalMeshLocation(Bone Sampling Mode=::NewEnumerator4, MaskSpawnGroup=false, Skeletal Mesh=$User.Skeletal Mesh)
   &ParticleState()
   &ScaleMeshSize(Scale Factor=&Multiply_VectorByFloat(Vector=<1 1 1>, Float=&RandomRangeFloat(Minimum=0.8, Maximum=1)))
   &ScaleColor(
    Scale Alpha=&Multiply_Float(A=20, B=&FloatFromCurve(FloatCurve=$System.AlphaCurve))
    Scale RGB=&Multiply_VectorByFloat(
     Vector=<1 1 1>
     Float=&Multiply_Float(
      A=&FloatFromCurve(FloatCurve=$System.EmissiveCurve)
      B=&Multiply_Float(A=$System.Tod, B=1)
     )
    )
    ScaleA=true
    ScaleRGB=true
   )
   &DynamicMaterialParameters(
    Index 0 Param 1=&FloatFromCurve(FloatCurve=DataInterface() {propertyValues {Curve {keys [
         {arriveTangent=0.4234, interpMode=RCIM_Cubic, leaveTangent=0.4234, tangentMode=RCTM_User, time=0, value=0}
         {arriveTangent=2.3288, interpMode=RCIM_Cubic, leaveTangent=2.3288, tangentMode=RCTM_User, time=1, value=1}
        ]}}})
    Index 0 Param 2=0.7
    Index 0 Param 3=&FloatFromCurve(FloatCurve=DataInterface() {propertyValues {Curve {keys [{tangentMode=RCTM_User, time=0, value=0}, {tangentMode=RCTM_User, time=1, value=5}]}}})
    Index 0 Param 4=0.7
    Index 1 Param 1=&FloatFromCurve(FloatCurve=DataInterface() {propertyValues {Curve {keys [<at 0 1>,<at 1 0.3>]}}}, Scale Curve=0.3)
    Index 1 Param 3=$System.Lightning_U_Tile
    Module.Param0WriteEnabled=true
    Module.Param0WriteEnabled1=true
    Module.Param1WriteEnabled=true
    Module.Param1WriteEnabled1=true
    Module.Param2WriteEnabled=true
    Module.Param2WriteEnabled1=true
    Module.Param3WriteEnabled=true
    Module.Param3WriteEnabled1=true
    Write Parameter Index 1=true
   ) {enabled=false}
   &DynamicMaterialParameters(
    Index 0 Param 1=1
    Index 0 Param 2=0.7
    Index 0 Param 4=0.7
    Index 1 Param 1=&FloatFromCurve(FloatCurve=DataInterface() {propertyValues {Curve {keys [<at 0.5 1>,<at 1 0.3>]}}}, Scale Curve=0.3)
    Index 1 Param 3=$System.Lightning_U_Tile
    Index 1 Param 4=1
    Index 2 Param 1=&FloatFromCurve(FloatCurve=$System.LeadEdgeCurve)
    Module.Param0WriteEnabled=true
    Module.Param0WriteEnabled1=true
    Module.Param0WriteEnabled2=true
    Module.Param1WriteEnabled=true
    Module.Param1WriteEnabled1=true
    Module.Param1WriteEnabled2=true
    Module.Param2WriteEnabled=true
    Module.Param2WriteEnabled1=true
    Module.Param2WriteEnabled2=true
    Module.Param3WriteEnabled=true
    Module.Param3WriteEnabled1=true
    Module.Param3WriteEnabled2=true
    Write Parameter Index 1=true
    Write Parameter Index 2=true
   ) {module=DynamicMaterialParameters001}
   &SolveForcesAndVelocity(Clamp Velocity=false, Limit Acceleration=false)
   /SetVariables(Particles.Glow Color=&MakeLinearColorFromVectorAndFloat(
     "Vector (RGB)"=&MakeVectorFromLinearColorRGB(Color=&MultiplyLinearColor(Linear Color=<rgba 0.5 0.65 1 1>, Scale Factor=&MakeVector4_FromFloat(Float=1)))
     "Float (Alpha)"=&Multiply_Float(
      A=&FloatFromCurve(
       FloatCurve=DataInterface() {propertyValues {Curve {keys [
           {interpMode=RCIM_Cubic, tangentMode=RCTM_User, time=0, value=1}
           <at 0.7 1>
           {arriveTangent=-3.6307, interpMode=RCIM_Cubic, leaveTangent=-3.6307, tangentMode=RCTM_User, time=0.9, value=0}
          ]}}}
       Scale Curve=$System.Tod
      )
      B=1.5
     )
    )) {isSetParameters=true, module=SetVariables_#10}
   /SetVariables(Particles.Glow Dy02:Vector4f=&MakeVector4(
     X=&FloatFromCurve(FloatCurve=DataInterface() {propertyValues {Curve {keys [
          {arriveTangent=0.4234, interpMode=RCIM_Cubic, leaveTangent=0.4234, tangentMode=RCTM_User, time=0, value=0}
          {arriveTangent=2.3288, interpMode=RCIM_Cubic, leaveTangent=2.3288, tangentMode=RCTM_User, time=0.9, value=1}
         ]}}})
     Y=0.7
     Z=0
     W=0.5
    )) {isSetParameters=true, module=SetVariables_#11}
   /SetVariables(Particles.Glow Dy03:Vector4f=&MakeVector4(
     X=&FloatFromCurve(FloatCurve=DataInterface() {propertyValues {Curve {keys [<at 0 1>,<at 1 0.5>]}}}, Scale Curve=0.5)
     Y=0
     Z=0
     W=0
    )) {isSetParameters=true, module=SetVariables_#12}
   &FN_Light_Attributes(
    Color=$System.ParticleLightColor
    Create Color Attribute=true
    Create Exponent Attribute=true
    Create Position Attribute=true
    Create Radius Attribute=true
    Exponent=2
    Position=&AddVectorToPosition(Position=$Particles.Position, Vector=&RandomRangeVector(Minimum=<5 -8 0>, Maximum=<-5 -8 5>))
    Radius=30
    Scale Color=&FloatFromCurve(FloatCurve=$System.ParticleLightCurve, Scale Curve=$System.Tod)
   )
  ]
  rendererClasses [&NiagaraMeshRendererProperties, &NiagaraLightRendererProperties]
  renderers [
   {rendererClass=&NiagaraMeshRendererProperties, rendererIndex=0}
   {rendererClass=&NiagaraMeshRendererProperties, rendererIndex=1}
   {rendererClass=&NiagaraLightRendererProperties, rendererIndex=2}
  ]
 }
 Lightning04 {
  emitterUpdateScript [
   &EmitterState(
    Life Cycle Mode=::NewEnumerator1
    Loop Delay=&RandomRangeFloat(Minimum=2, Maximum=5)
    Loop Duration=&RandomRangeFloat(Minimum=0.3, Maximum=0.5)
    Recalculate Duration Each Loop=true
    UseLoopDelay=true
   )
   &SpawnRate(Use Spawn Probability=false)
  ]
  particleSpawnScript [
   /SetVariables(Particles.RandomFloat=&RandomRangeFloat(Minimum=0.5, Maximum=1)) {isSetParameters=true, module=SetVariables_#6}
   &InitializeParticle(
    Color=$System.LightningColor
    Lifetime=&Multiply_Float(A=$System.LightningLifetime, B=$Particles.RandomFloat)
    Mesh Scale Mode=::NewEnumerator4
    UsePositionOffset=false
    Write Material Random=true
   )
   &InitialMeshOrientation(
    Mesh Orientation Mode=::NewEnumerator2
    Rotation=&MakeCustomVectorFromBool(Bool=&RandomBool(), True Vector=<0.5 0.1 0.1>, False Vector=<0 -0.35 0.1>)
    Use Rotation Vector=true
   )
   &CameraOffset()
  ]
  particleUpdateScript [
   &SkeletalMeshLocation(Bone Sampling Mode=::NewEnumerator4, MaskSpawnGroup=false, Sampled Position Offset=<0 -5 10>, Skeletal Mesh=$User.Skeletal Mesh)
   &ParticleState()
   &ScaleMeshSize(Scale Factor=&Multiply_VectorByFloat(Vector=<1 1 1>, Float=&RandomRangeFloat(Minimum=0.8, Maximum=1)))
   &ScaleColor(
    Scale Alpha=&Multiply_Float(A=20, B=&FloatFromCurve(FloatCurve=$System.AlphaCurve))
    Scale RGB=&Multiply_VectorByFloat(
     Vector=<1 1 1>
     Float=&Multiply_Float(
      A=&FloatFromCurve(FloatCurve=$System.EmissiveCurve)
      B=&Multiply_Float(A=$System.Tod, B=1)
     )
    )
    ScaleA=true
    ScaleRGB=true
   )
   &DynamicMaterialParameters(
    Index 0 Param 1=&FloatFromCurve(FloatCurve=DataInterface() {propertyValues {Curve {keys [
         {arriveTangent=0.4234, interpMode=RCIM_Cubic, leaveTangent=0.4234, tangentMode=RCTM_User, time=0, value=0}
         {arriveTangent=2.3288, interpMode=RCIM_Cubic, leaveTangent=2.3288, tangentMode=RCTM_User, time=1, value=1}
        ]}}})
    Index 0 Param 2=0.7
    Index 0 Param 3=&FloatFromCurve(FloatCurve=DataInterface() {propertyValues {Curve {keys [{tangentMode=RCTM_User, time=0, value=0}, {tangentMode=RCTM_User, time=1, value=5}]}}})
    Index 0 Param 4=0.7
    Index 1 Param 1=&FloatFromCurve(FloatCurve=DataInterface() {propertyValues {Curve {keys [<at 0 1>,<at 1 0.3>]}}}, Scale Curve=0.3)
    Index 1 Param 3=$System.Lightning_U_Tile
    Module.Param0WriteEnabled=true
    Module.Param0WriteEnabled1=true
    Module.Param1WriteEnabled=true
    Module.Param1WriteEnabled1=true
    Module.Param2WriteEnabled=true
    Module.Param2WriteEnabled1=true
    Module.Param3WriteEnabled=true
    Module.Param3WriteEnabled1=true
    Write Parameter Index 1=true
   ) {enabled=false}
   &DynamicMaterialParameters(
    Index 0 Param 1=1
    Index 0 Param 2=0.7
    Index 0 Param 4=0.7
    Index 1 Param 1=&FloatFromCurve(FloatCurve=DataInterface() {propertyValues {Curve {keys [<at 0.5 1>,<at 1 0.3>]}}}, Scale Curve=0.3)
    Index 1 Param 3=$System.Lightning_U_Tile
    Index 1 Param 4=1
    Index 2 Param 1=&FloatFromCurve(FloatCurve=$System.LeadEdgeCurve)
    Module.Param0WriteEnabled=true
    Module.Param0WriteEnabled1=true
    Module.Param0WriteEnabled2=true
    Module.Param1WriteEnabled=true
    Module.Param1WriteEnabled1=true
    Module.Param1WriteEnabled2=true
    Module.Param2WriteEnabled=true
    Module.Param2WriteEnabled1=true
    Module.Param2WriteEnabled2=true
    Module.Param3WriteEnabled=true
    Module.Param3WriteEnabled1=true
    Module.Param3WriteEnabled2=true
    Write Parameter Index 1=true
    Write Parameter Index 2=true
   ) {module=DynamicMaterialParameters001}
   &SolveForcesAndVelocity(Clamp Velocity=false, Limit Acceleration=false)
   /SetVariables(Particles.Glow Color=&MakeLinearColorFromVectorAndFloat(
     "Vector (RGB)"=&MakeVectorFromLinearColorRGB(Color=&MultiplyLinearColor(Linear Color=<rgba 0.5 0.65 1 1>, Scale Factor=&MakeVector4_FromFloat(Float=1)))
     "Float (Alpha)"=&Multiply_Float(
      A=&FloatFromCurve(
       FloatCurve=DataInterface() {propertyValues {Curve {keys [
           {interpMode=RCIM_Cubic, tangentMode=RCTM_User, time=0, value=1}
           <at 0.7 1>
           {arriveTangent=-3.6307, interpMode=RCIM_Cubic, leaveTangent=-3.6307, tangentMode=RCTM_User, time=0.9, value=0}
          ]}}}
       Scale Curve=$System.Tod
      )
      B=1.5
     )
    )) {isSetParameters=true}
   /SetVariables(Particles.Glow Dy02:Vector4f=&MakeVector4(
     X=&FloatFromCurve(FloatCurve=DataInterface() {propertyValues {Curve {keys [
          {arriveTangent=0.4234, interpMode=RCIM_Cubic, leaveTangent=0.4234, tangentMode=RCTM_User, time=0, value=0}
          {arriveTangent=2.3288, interpMode=RCIM_Cubic, leaveTangent=2.3288, tangentMode=RCTM_User, time=1, value=1}
         ]}}})
     Y=0.7
     Z=0
     W=0.5
    )) {isSetParameters=true}
   /SetVariables(Particles.Glow Dy03:Vector4f=&MakeVector4(
     X=&FloatFromCurve(FloatCurve=DataInterface() {propertyValues {Curve {keys [<at 0 1>,<at 1 0.5>]}}}, Scale Curve=0.5)
     Y=0
     Z=0
     W=0
    )) {isSetParameters=true}
   &FN_Light_Attributes(
    Color=$System.ParticleLightColor
    Create Color Attribute=true
    Create Exponent Attribute=true
    Create Position Attribute=true
    Create Radius Attribute=true
    Exponent=2
    Position=&AddVectorToPosition(Position=$Particles.Position, Vector=&RandomRangeVector(Minimum=<5 -10 0>, Maximum=<-5 -10 5>))
    Radius=30
    Scale Color=&FloatFromCurve(FloatCurve=$System.ParticleLightCurve, Scale Curve=$System.Tod)
   )
  ]
  rendererClasses [&NiagaraMeshRendererProperties, &NiagaraLightRendererProperties]
  renderers [
   {rendererClass=&NiagaraMeshRendererProperties, rendererIndex=0}
   {rendererClass=&NiagaraMeshRendererProperties, rendererIndex=1}
   {rendererClass=&NiagaraLightRendererProperties, rendererIndex=2}
  ]
 }
}
summary {
 systemName=NS_Backpack_FishBowlBone
 emitters [
  {emitterName=Lightning01, rendererClasses [&NiagaraMeshRendererProperties, &NiagaraLightRendererProperties]}
  {emitterName=Lightning02, rendererClasses [&NiagaraMeshRendererProperties, &NiagaraLightRendererProperties]}
  {emitterName=Lightning03, rendererClasses [&NiagaraMeshRendererProperties, &NiagaraLightRendererProperties]}
  {emitterName=Lightning04, rendererClasses [&NiagaraMeshRendererProperties, &NiagaraLightRendererProperties]}
  {emitterName=Lightning03001, rendererClasses [&NiagaraMeshRendererProperties, &NiagaraLightRendererProperties]}
  {
   bEnabled=false
   emitterName=Lightning01001
   rendererClasses [&NiagaraMeshRendererProperties, &NiagaraLightRendererProperties]
  }
 ]
 userVariables [
  User.bIsFrontEnd {type=NiagaraBool}
  User.bIsFrontEndPreview {type=NiagaraBool, defaultValue=true}
  User.Skeletal Mesh {
   type=NiagaraDataInterfaceSkeletalMesh
   defaultValue {
    struct=NiagaraExt_VariableValue_DataInterface
    value {dataInterface=&self:NiagaraDataInterfaceSkeletalMesh_0}
   }
  }
 ]
}
